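# Skeletal-mesh character material for the Marshmellow DJ body, built on the shared character function stack with a Marshmello skin effect.
# Base PreFX attributes pass through the Marshmello function, borrowed from weapon skins, then into PostFX.
# Static component mask parameters select SkinFX_Mask channels for the effect's alpha, offset and low/high EQ inputs.
# PreFX AO also drives PostFX's AO, distance blend and fresnel inputs.
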Material /BRCosmetics/Characters/Player/Male/Medium/Bodies/M_MED_Marshmellow_DJ/Materials/M_Marshmello_DJ_Body_NewMaster
graph {
 properties {
  bUseMaterialAttributes=true
  bUsedWithMeshDeformer=true
  bUsedWithMorphTargets=true
  bUsedWithSkeletalMesh=true
  pixelDepthOffsetMode=PDOM_Legacy
  usedShadingModels=Default Lit
 }
 nodes [
  n1 = MaterialFunctionCall() {materialFunction=&BaseFN_CharacterFunction_PreFX}
  n2 = TextureSample(Tex=TextureObjectParameter() {group=Skin FX, parameterName=SkinFX_Mask}) {samplerType=SAMPLERTYPE_Masks}
  MaterialFunctionCall(
   AO=@n1.AO
   Distance Blend=@n1.AO
   Fresnel=@n1.AO
   Material=MaterialFunctionCall!(
    Alpha=StaticComponentMaskParameter(None=@n2.RGBA) {defaultR=true, group=Skin FX, parameterName=Marshmello Channel}
    Offset=StaticComponentMaskParameter(None=@n2.RGBA) {defaultG=true, group=Skin FX, parameterName=Marshmello Offset Channel}
    EQ Low=StaticComponentMaskParameter(None=@n2.RGBA) {defaultR=true, group=Skin FX, parameterName=Marshmello Low Channel}
    EQ High=StaticComponentMaskParameter(None=@n2.RGBA) {defaultG=true, group=Skin FX, parameterName=Marshmello High Channel}
    Material=@n1.Result
   ) {materialFunction=/Game/Weapons/WeaponSkins/MaterialFunction/Marshmello}.Result
  ) {materialFunction=&BaseFN_CharacterFunction_PostFX}
 ]
}
