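# Physics asset of body colliders for the medium male Marshmellow DJ player character, named as collision for its cape.
# Capsules cover the head, neck, spine, clavicles, upper arms, pelvis, thighs and calves, with no forearm, hand or foot bodies.
# The pelvis carries two capsules, one offset and tilted, and the head uses a zero-length capsule as a sphere.
# It defines no constraints, so the shapes are not jointed as a ragdoll.
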PhysicsAsset /BRCosmetics/Characters/Player/Male/Medium/Bodies/M_MED_Marshmellow_DJ/Meshes/M_MED_Marshmellow_CapeColliders
bodies {
 calf_l [{shapeType=Capsule, length=31.242, radius=11.9116, center=<-19.4843 -0.5441 -0.2294>, rotation=<rot 0.7422 94.0861 -89.4657>}]
 calf_r [{shapeType=Capsule, length=31.5362, radius=11.5971, center=<19.4486 0.3664 0.2406>, rotation=<rot 0.7421 93.6526 -89.373>}]
 clavicle_l [{shapeType=Capsule, length=10.3261, radius=8.5802, center=<8.2822 -0.2527 -2.0684>, rotation=<rot 90 -0.0049 179.9951>}]
 clavicle_r [{shapeType=Capsule, length=10.3261, radius=8.5802, center=<-7.9706 0.7648 2.6048>, rotation=<rot 90 -0.0099 -0.0099>}]
 head [{shapeType=Capsule, length=0, radius=18, center=<11.6361 0.0962 0>, rotation=<rot 0 -90 0>}]
 neck_02 [{shapeType=Capsule, length=0.505, radius=10.755, rotation=<rot 0 -90 0>}]
 pelvis [
  {shapeType=Capsule, length=11.4559, radius=14.3329, center=<2.2747 -0.2766 1.0154>}
  {shapeType=Capsule, length=2.75, radius=10.5, center=<-2.8984 -11.8113 -16.5676>, rotation=<rot 0 0 -27.9568>}
 ]
 spine_03 [{shapeType=Capsule, length=16.5169, radius=16.5072, center=<1.1872 0.9656 0.153>, rotation=<rot -83.7233 -83.7199 81.9245>}]
 spine_05 [{shapeType=Capsule, length=15.7232, radius=13.517, center=<-0.8162 -0.0647 0>}]
 thigh_l [{shapeType=Capsule, length=26.002, radius=10.5818, center=<-19.839 -0.0002 0>, rotation=<rot 89.2574 89.9982 89.9982>}]
 thigh_r [{shapeType=Capsule, length=25.6153, radius=10.4315, center=<19.0128 0 0>, rotation=<rot 89.2574 89.9982 89.9982>}]
 upperarm_l [{shapeType=Capsule, length=18.8261, radius=7.5802, center=<12.1824 -0.2526 -0.6865>, rotation=<rot 90 0.005 -179.995>}]
 upperarm_r [{shapeType=Capsule, length=17.1357, radius=7.1339, center=<-12.8316 -0.0074 -0.1319>, rotation=<rot 0 -88.4734 -86.4203>}]
}
constraints []
