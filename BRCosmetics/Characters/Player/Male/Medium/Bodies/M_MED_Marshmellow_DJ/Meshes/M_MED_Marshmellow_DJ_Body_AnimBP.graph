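# Post-body animation Blueprint for the Marshmellow DJ medium male player body cosmetic, adding secondary motion to clips and grenades.
# Copies the pose from the attached parent mesh, then positions lowClip and topClip with weighted multi-bone constraints on spine, pelvis, clavicle and arm bones.
# Chained AnimDynamics nodes then swing the back and front grenade bones with damped angular springs, tight limits and root-relative simulation.
# The event graph has no logic.
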AnimBlueprint /BRCosmetics/Characters/Player/Male/Medium/Bodies/M_MED_Marshmellow_DJ/Meshes/M_MED_Marshmellow_DJ_Body_AnimBP
target_skeleton=&/M_MED_Marshmellow_DJ_Skeleton
graphs [
 AnimGraph {nodes [
   n1 = "Constraint - lowClip (PPPPPPPP)"!(
    ComponentPose=Local To Component(LocalPose=Copy Pose From Mesh() {bUseAttachedParent=true})
    Alpha=1
    ConstraintWeights_0=0.085
    ConstraintWeights_1:float=0.02
    ConstraintWeights_2:float=0.061
    ConstraintWeights_3:float=0.026
    ConstraintWeights_4:float=0.148
    ConstraintWeights_5:float=0.226
    ConstraintWeights_6:float=0.379
    ConstraintWeights_7:float=0.056
   ) -> (Pose) {
    alphaBoolBlend {blendOption=Linear}
    boneToModify {boneName=lowClip}
    constraintSetup [
     {targetBone {boneName=clavicle_l}}
     {targetBone {boneName=clavicle_r}}
     {targetBone {boneName=pelvis}}
     {targetBone {boneName=spine_01}}
     {targetBone {boneName=spine_02}}
     {targetBone {boneName=spine_03}}
     {targetBone {boneName=spine_04}}
     {targetBone {boneName=spine_05}}
    ]
    constraintWeights [1,1,1,1,1,1,1,1]
   }
   n2 = "Constraint - topClip (PPPPPPPP)"!(
    ComponentPose=@n1
    Alpha=1
    ConstraintWeights_0=0.239
    ConstraintWeights_1:float=0.025
    ConstraintWeights_2:float=0.038
    ConstraintWeights_3:float=0.023
    ConstraintWeights_4:float=0.099
    ConstraintWeights_5:float=0.194
    ConstraintWeights_6:float=0.296
    ConstraintWeights_7:float=0.085
   ) -> (Pose) {
    alphaBoolBlend {blendOption=Linear}
    boneToModify {boneName=topClip}
    constraintSetup [
     {targetBone {boneName=clavicle_l}}
     {targetBone {boneName=upperarm_twist_01_l}}
     {targetBone {boneName=pelvis}}
     {targetBone {boneName=spine_01}}
     {targetBone {boneName=spine_02}}
     {targetBone {boneName=spine_03}}
     {targetBone {boneName=spine_04}}
     {targetBone {boneName=spine_05}}
    ]
    constraintWeights [1,1,1,1,1,1,1,1]
   }
   n3 = Anim Dynamics - Bone: dyn_Grenade_Back(ComponentPose=@n2, Alpha=1, ExternalForce="0,0,0") {
    angularDampingOverride=0.9999
    angularSpringConstant=900
    bAngularSpring=true
    bGravityOverrideInSimSpace=true
    bOverrideAngularDamping=true
    bOverrideLinearDamping=true
    simulationSpace=RootRelative
    alphaBoolBlend {blendOption=Linear}
    boundBone {boneName=dyn_Grenade_Back}
    componentAppliedLinearAccClamp=<400 400 400>
    componentLinearAccScale=<0.15 0.15 0.15>
    componentLinearVelScale=<0.15 0.15 0.15>
    physicsBodyDefinitions [{
      sphereCollisionRadius=0
      boundBone {boneName=dyn_Grenade_Back}
      boxExtents=<30 30 5>
      constraintSetup {angularTargetAxis=AxisZ, twistAxis=AxisZ, angularLimitsMax=<10 10 10>, angularLimitsMin=<-10 -10 -10>, angularTarget=<0 0 0>}
      localJointOffset=<-0.2 0 -5>
     }]
   }
   n4 = Anim Dynamics - Bone: dyn_Grenade_Front(ComponentPose=@n3, Alpha=1, ExternalForce="0,0,0") {
    angularDampingOverride=0.99
    angularSpringConstant=1000
    bAngularSpring=true
    bGravityOverrideInSimSpace=true
    bOverrideAngularDamping=true
    bOverrideLinearDamping=true
    simulationSpace=RootRelative
    alphaBoolBlend {blendOption=Linear}
    boundBone {boneName=dyn_Grenade_Front}
    componentAppliedLinearAccClamp=<400 400 400>
    componentLinearAccScale=<0.15 0.15 0.15>
    componentLinearVelScale=<0.15 0.15 0.15>
    physicsBodyDefinitions [{
      sphereCollisionRadius=0
      boundBone {boneName=dyn_Grenade_Front}
      boxExtents=<40 40 8>
      constraintSetup {angularTargetAxis=AxisZ, twistAxis=AxisZ, angularLimitsMax=<10 10 10>, angularLimitsMin=<-10 -10 -10>, angularTarget=<0 0 0>}
      localJointOffset=<-0.2 0 -5>
     }]
   }
   Output Pose(Result=Component To Local(ComponentPose=@n4)) {layerGroup=None}
  ]}
 EventGraph {nodes [Event BlueprintUpdateAnimation()]}
]
variables []
