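# Animation Blueprint for the face accessory part of the Marshmallow DJ cosmetic character.
# The output pose copies pose and curves directly from the attached parent mesh, so the part follows the main body.
# A cached "HeadLockCache" pose and a layered blend filtered at neck_01 remain in the graph but do not feed the output.
# The event graph holds an update event with no logic.
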AnimBlueprint /BRCosmetics/Characters/Player/Male/Medium/Bodies/M_MED_Marshmellow_DJ/Meshes/Parts/M_MED_Marshmallow_DJ_FaceAcc_AnimBP
target_skeleton=&/M_MED_Marshmallow_DJ_FaceAcc_Skeleton
graphs [
 AnimGraph {nodes [
   Layered blend per bone!(
    BasePose=Use cached pose 'HeadLockCache'()
    BlendPoses_0
    BlendPoses_1:Pose Link Structure=Use cached pose 'HeadLockCache'()
    BlendWeights_0=0.735
    BlendWeights_1:float=1
   ) -> (Pose) {
    blendPoses [{linkId=-1, sourceLinkId=-1}, {linkId=-1, sourceLinkId=-1}]
    blendWeights [1,1]
    layerSetup [{branchFilters [{blendDepth=1, boneName=neck_01}]}, {branchFilters []}]
   }
   Output Pose(Result=Copy Pose From Mesh() {bCopyCurves=true, bUseAttachedParent=true}) {layerGroup=None}
   Save cached pose 'HeadLockCache'() {cacheName=HeadLockCache}
  ]}
 EventGraph {nodes [Event BlueprintUpdateAnimation()]}
]
variables []
